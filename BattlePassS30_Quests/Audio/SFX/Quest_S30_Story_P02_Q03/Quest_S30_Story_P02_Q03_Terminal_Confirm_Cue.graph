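# Sound cue for the terminal confirm sound in a Season 30 story quest (P02 Q03).
# A mixer layers a random pick from three confirm waves with a quieter, briefly delayed reward-screen chest upgrade sting.
# It plays as a gadget-class sound with borrowed supply drop attenuation.
# A concurrency override allows one instance per owner and blocks new plays while one is active.
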SoundCue /BattlePassS30_Quests/Audio/SFX/Quest_S30_Story_P02_Q03/Quest_S30_Story_P02_Q03_Terminal_Confirm_Cue
root=Mixer_0
nodes [
 {id=Delay_0, children [WavePlayer_5], properties {delayMax=0.2, delayMin=0.2}}
 {id=Mixer_0, children [Random_0,Delay_0], properties {inputVolume [1,0.4]}}
 {id=Random_0, children [WavePlayer_1,WavePlayer_2,WavePlayer_0], properties {weights [1,1,1]}}
 {
  id=WavePlayer_0
  properties {soundWaveAssetPtr=&/Quest_S30_Story_P02_Q03_Terminal_Confirm_03}
 }
 {
  id=WavePlayer_1
  properties {soundWaveAssetPtr=&/Quest_S30_Story_P02_Q03_Terminal_Confirm_01}
 }
 {
  id=WavePlayer_2
  properties {soundWaveAssetPtr=&/Quest_S30_Story_P02_Q03_Terminal_Confirm_02}
 }
 {id=WavePlayer_5, properties {soundWaveAssetPtr=/Game/Sounds/Fort_UI_Sounds/Rewards/Fort_RewardScreen_TeamScore_ChestLootUpgrade}}
]
properties {
 attenuationSettings=/Game/Athena/Sounds/SupplyDrops/Fort_HiddenStar_Ambient_Loop_Att
 bOverrideConcurrency=true
 maxDistance=2500
 soundClassObject=&Fort_Gadgets
 volumeMultiplier=0.5
 concurrencyOverrides {bLimitToOwner=true, maxCount=1, resolutionRule=PreventNew}
}
